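# One-shot mono MetaSound template for Mimosa cosmetic companion voice lines, with empty wave array and pet sound class.
# On play, a probability gate leads to a random delay, then a no-repeat pick from the wave array via a shared patch.
# The chosen wave plays with randomized pitch and decibel gain.
# A failed probability roll routes through a trigger variable to OnFinished, so the source ends without playing.
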MetaSoundSource /CosmeticCompanions/Assets/Audio_Shared/VO/MSS_Mimosa_VO_Template_Mono
inputs [
 "Delay Max (Seconds)" {type=Time}
 "Delay Min (Seconds)" {type=Time}
 "Probability (0-1)" {type=Float}
 "Random Pitch Max (Semitones)" {type=Float}
 "Random Pitch Min (Semitones)" {type=Float}
 "Random Volume Change Max (Decibels)" {type=Float}
 "Random Volume Change Min (Decibels)" {type=Float}
 UE.Source.OnPlay {type=Trigger}
 Wave Array {type=WaveAsset:Array}
]
interfaces [UE.OutputFormat.Mono:1.0, UE.Source.OneShot:1.0, UE.Source:1.1]
outputs [
 On Play {type=Trigger}
 "UE.OutputFormat.Mono.Audio:0" {type=Audio}
 UE.Source.OneShot.OnFinished {type=Trigger}
]
pages [Default {
  input_defaults {
   "Delay Max (Seconds)"=0
   "Delay Min (Seconds)"=0
   "Probability (0-1)"=1
   "Random Pitch Max (Semitones)"=0
   "Random Pitch Min (Semitones)"=0
   "Random Volume Change Max (Decibels)"=0
   "Random Volume Change Min (Decibels)"=0
   UE.Source.OnPlay=null
   Wave Array []
  }
  nodes [
   TriggerFilter_0 = UE.Trigger Filter(Probability=^"Probability (0-1)", Trigger=^UE.Source.OnPlay)
   n1 = UE.RandomTime(Max=^"Delay Max (Seconds)", Min=^"Delay Min (Seconds)", Next=@TriggerFilter_0.Heads)
   TriggerDelay_0 = UE.Trigger Delay(In=@n1.On Next, Delay Time=@n1.Value)
   n2 = /FortMetasound/Patches/MSP_GetNoRepeatNumber(WaveArray=^Wave Array, InTrigger=@TriggerDelay_0.Out)
   ArrayRandomGet_0 = Array.Random Get:WaveAsset:Array(In Array=^Wave Array, No Repeats=@n2.NoRepeatNumber, Next=@n2.OutTrigger) {also {defaults {Enable Shared State="<MetaSoundBuilderResult.SUCCEEDED: 0>"}}}
   n3 = UE.RandomFloat(Max=^"Random Pitch Max (Semitones)", Min=^"Random Pitch Min (Semitones)", Next=@ArrayRandomGet_0.On Next)
   n4 = UE.RandomFloat(Max=^"Random Volume Change Max (Decibels)", Min=^"Random Volume Change Min (Decibels)", Next=@n3.On Next)
   DecibelstoLinearGain_0 = UE.Decibels to Linear Gain:Float(Decibels=@n4.Value)
   WavePlayer_0 = UE.Wave Player:Mono(Wave Asset=@ArrayRandomGet_0.Value, Play=@n4.On Next, Pitch Shift=UE.Reroute(Value=@n3.Value).Value) {also {version [1,1]}}
   Multiply_0 = UE.Multiply:Audio by Float(AdditionalOperands=@DecibelstoLinearGain_0.Linear Gain, PrimaryOperand=@WavePlayer_0.Out Mono)
   n5 = InitVariable.Trigger()
   TriggerAnyTriggerAny2_0 = "TriggerAny.Trigger Any (2)"(
    In 1=VariableAccessor.Trigger(Variable=VariableMutator.Trigger(Variable=@n5.Variable, Value=@TriggerFilter_0.Tails).Variable).Value
    In 0=@WavePlayer_0.On Finished
   )
  ]
  links [
   {dst="@UE.OutputFormat.Mono.Audio:0", src=Multiply_0.Out}
   {dst="@UE.Source.OneShot.OnFinished", src=TriggerAnyTriggerAny2_0.Out}
   {dst="@On Play", src=WavePlayer_0.On Play}
  ]
 }]
properties {
 assetImportData=&self:AssetImportData
 attenuationSettings=/Game/Sounds/Attenuation/AS_Companion_VO
 audioPropertiesSheet=/CosmeticCompanions/Assets/Audio_Shared/PropertySheets/APS_Cosmetic_Mimosa_VO
 maxDistance=2200
 sampleRate=48000
 soundClassObject=&Fort_Creatures_Pets
 assetUserData [&self:AssetUserData_0]
 concurrencyOverrides {bLimitToOwner=true, maxCount=1, resolutionRule=PreventNew}
 concurrencySet [/Game/Sounds/Concurrency_Rules/CS_Mimosa_VO]
}
